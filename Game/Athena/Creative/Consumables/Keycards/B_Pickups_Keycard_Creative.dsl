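; Creative keycard pickup Blueprint that gives each keycard its material based on its item definition.
; On BeginPlay it looks up the item definition in a material map and, where cosmetic events can run, async loads the match, applies it to the mesh primitive, then sets rotation and scale.
; On pickup it creates a dynamic material instance and adds it to the parent's MID array.
; BeginPlay calls its parent function twice.
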(event EventBeginPlay
  (bind _itemdefinition (Variables|PickupProperties|GetItemDefinition))
  (bind _item_definition_materials (Variables|Default|GetItem%20Definition%20Materials))
  (bind (_value _returnvalue) (Utilities|Map|Find _item_definition_materials _itemdefinition))
  (bind _skeletalorstaticmeshassetprimitive (Variables|Default|GetSkeletalOrStaticMeshAssetPrimitive))
  (|CallParentFunction)
  (|CallParentFunction)
  (Networking|CanExecuteCosmeticEvents
    (:True
      (if _returnvalue
        (bind _object (Utilities|AsyncLoadAsset _value)
          (:Completed
            (bind _asmaterial_interface (Utilities|Casting|CastToMaterialInterface _object))
            (Rendering|Material|SetMaterial _skeletalorstaticmeshassetprimitive 0 _asmaterial_interface)
            (Transformation|SetWorldRotation _skeletalorstaticmeshassetprimitive (Variables|Default|GetRotation_SM_AssetPrimitive) false false)
            (Transformation|SetWorldScale3D _skeletalorstaticmeshassetprimitive (Variables|Default|GetScale_SM_AssetPrimitive))))))))

(event EventOnPickedUp
  (params (PickupTarget :type Fort%20Pawn%20Object%20Reference))
  (|CallParentFunction 0)
  (Networking|CanExecuteCosmeticEvents
    (:True
      (bind _returnvalue (Rendering|Material|CreateDynamicMaterialInstance (Variables|Default|GetSkeletalOrStaticMeshAssetPrimitive) 0 0 "None" :!class /Script/Engine.PrimitiveComponent))
      (Utilities|Array|AddUnique (Variables|FortPickupsParent|GetMIDs) _returnvalue))))

(fn ConstructionScript
  (|CallParentFunction))
